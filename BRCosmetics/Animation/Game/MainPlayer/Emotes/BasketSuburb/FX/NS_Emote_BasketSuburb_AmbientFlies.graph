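# Niagara system of ambient flies for the BasketSuburb player emote.
# DirectionalBurst swarms animated sprite flies within a User.Radius sphere using curl noise, spawning faster when the owner moves and calmer in the front end.
# BlowingParticles reads those flies through attribute readers to draw short, fading ribbon trails behind them.
# Color adapts to quality level, with distance and time of day tinting the flies and indoor state darkening both emitters at higher quality.
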NiagaraSystem /BRCosmetics/Animation/Game/MainPlayer/Emotes/BasketSuburb/FX/NS_Emote_BasketSuburb_AmbientFlies
dependencies {
 usedDataInterfaces [
  &NiagaraDataInterfaceCurve
  &NiagaraDataInterfaceVectorCurve
  &NiagaraDataInterfaceSpriteRendererInfo
  &NiagaraDataInterfaceParticleRead
  &NiagaraDataInterfaceColorCurve
 ]
 usedDynamicInputs [
  &MakeVector4_FromLinearColor
  /Niagara/DynamicInputs/LinearColor/MakeLinearColorFromVector4
  &Vector4FromCurve
 ]
 usedModules [&SystemState]
}
emitters {
 BlowingParticles {
  emitterUpdateScript [
   /SetVariables(Emitter.NewSpawnRate=16) {isSetParameters=true}
   &EmitterState(Life Cycle Mode=::NewEnumerator1, Loop Duration=0.0056, UseLoopDelay=false)
   &SpawnParticlesFromOtherEmitter(
    Attribute Reader=DataInterface() {propertyValues {EmitterBinding {bindingMode=Other, emitterName=DirectionalBurst}}}
    Maximum Age Spawn Threshold=1.4
    Spawn Rate=$Emitter.NewSpawnRate
    Use Maximum Age=false
    Use Minimum Age=false
   )
  ]
  particleSpawnScript [
   &InitializeParticle(
    Color=&FNTimeOfDayColor(Night Color=<rgba 0.4669 0.5366 1 0.113>, Morning Color=<rgba 1 1 1 0.113>, Day Color=<rgba 1 1 1 0.113>, Evening Color=<rgba 1 1 1 0.113>)
    Lifetime=0.234
    Ribbon Width=1.0428
    Ribbon Width Mode=::NewEnumerator1
    UsePositionOffset=false
    Write Material Random=true
   )
   &SampleParticlesFromOtherEmitter(
    Particle ID Sampling=::NewEnumerator1
    Reject Particle If Sampled Index Is Invalid=false
    Sprite Size Sampling:ENiagara_DirectReadSamplingMode=::NewEnumerator1
   )
  ]
  particleUpdateScript [
   &ParticleState()
   /Niagara/Modules/AttributeReader/UpdateParticlesFromOtherEmitter(
    Attribute Reader=DataInterface() {propertyValues {EmitterBinding {bindingMode=Other, emitterName=DirectionalBurst}}}
    Position Sampling:ENiagara_DirectReadSamplingMode=::NewEnumerator1
    Sprite Size Sampling:ENiagara_DirectReadSamplingMode=::NewEnumerator1
   )
   &self:ScratchModule(Position:NiagaraPosition=$Output.UpdateParticlesFromOtherEmitter.SampledPosition) {module=ScratchModule}
   &ScaleRibbonWidth(Ribbon Width Scale=&Multiply_Float(
     A=&FloatFromCurve(FloatCurve=DataInterface() {propertyValues {Curve {keys [<at 0 1>,<at 1 0>]}}}, Scale Curve=0.4)
     B=&MakeFloatFromVector2d(Vector 2D=$Output.UpdateParticlesFromOtherEmitter.SampledSpriteSize)
    ))
   &InheritSourceMovement(Applied Position Scale=<0.3 0.3 0.3>, ApplyForce=false, ApplyPosition=true, ApplyPositionFalloff=false)
   &SolveForcesAndVelocity(Clamp Velocity=false, Limit Acceleration=false)
   &InterpolateOverTime(
    Rate Of Change=5
    Target Float Value:NiagaraFloat=&RemapRange(Input Value To Remap=$NPC.FortniteNPC.IsPlayerIndoors, Input Min=0, Input Max=0.2, Output Min=0, Output Max=1)
   )
   &ScaleColor(
    Scale Alpha=&FloatFromCurve(FloatCurve=DataInterface() {propertyValues {Curve {keys [
         {arriveTangent=-0.8321, interpMode=RCIM_Cubic, leaveTangent=35.6275, tangentMode=RCTM_Break, time=0, value=0}
         {arriveTangent=0.0819, interpMode=RCIM_Cubic, leaveTangent=0.6334, tangentMode=RCTM_Break, time=0.0597, value=1}
         {arriveTangent=-1.7699, tangentMode=RCTM_User, time=1, value=0}
        ]}}})
    Scale RGB=&MergeHelper_VectorSwitch(
     0=&MakeVectorFromLinearColorRGB(Color=<rgba 1 1 1 1>)
     1=&MakeVectorFromLinearColorRGB(Color=<rgba 1 1 1 1>)
     2=&Lerp_Vector(
      A=&MakeVectorFromLinearColorRGB(Color=<rgba 1 1 1 1>)
      B=&MakeVectorFromLinearColorRGB(Color=<rgba 0.1719 0.1719 0.1719 1>)
      Alpha=$Particles.InterpolateOverTime.InterpolateOverTime_Float.Moving Average
     )
     3=&Lerp_Vector(
      A=&MakeVectorFromLinearColorRGB(Color=<rgba 1 1 1 1>)
      B=&MakeVectorFromLinearColorRGB(Color=<rgba 0.1719 0.1719 0.1719 1>)
      Alpha=$Particles.InterpolateOverTime.InterpolateOverTime_Float.Moving Average
     )
     4=&Lerp_Vector(
      A=&MakeVectorFromLinearColorRGB(Color=<rgba 1 1 1 1>)
      B=&MakeVectorFromLinearColorRGB(Color=<rgba 0.1719 0.1719 0.1719 1>)
      Alpha=$Particles.InterpolateOverTime.InterpolateOverTime_Float.Moving Average
     )
     Index=$Engine.QualityLevel
     Num=::NewEnumerator3
    )
    ScaleA=true
    ScaleRGB=true
   )
  ]
  rendererClasses [&NiagaraRibbonRendererProperties]
  renderers [{rendererClass=&NiagaraRibbonRendererProperties}]
 }
 DirectionalBurst {
  emitterUpdateScript [
   &self:InterpolateSpeedThresholdEmitter_01(
    Define Initial Value=false
    Rate Of Change=5
    Target Float Value:NiagaraFloat=&MakeCustomFloatFromBool(
     Bool=&SetBoolByFloatComparison(
      A=&Add_Float(
       A=&Add_Float(
        A=&Abs_Float(Float:NiagaraFloat=&MakeFloatFromVector(Vector=$Engine.Owner.Velocity, Channel=::NewEnumerator0))
        B=&Abs_Float(Float:NiagaraFloat=&MakeFloatFromVector(Vector=$Engine.Owner.Velocity, Channel=::NewEnumerator2))
       )
       B=&Abs_Float(Float:NiagaraFloat=&MakeFloatFromVector(Vector=$Engine.Owner.Velocity, Channel=::NewEnumerator1))
      )
      B=46.278
     )
     True Float=1
     False Float=0
    )
   ) {module=InterpolateSpeedThresholdEmitter_01}
   &EmitterState(Life Cycle Mode=::NewEnumerator1, Loop Behavior=::NewEnumerator1, Loop Duration Mode=::NewEnumerator1, UseLoopDelay=false)
   &SpawnRate(
    SpawnRate=&Lerp_Float(A=12, B=25, Alpha=$Emitter.InterpolateSpeedThresholdEmitter_01.InterpolateOverTime_Float.Moving Average)
    Use Spawn Probability=false
   )
  ]
  particleSpawnScript [
   &InitializeParticle(
    Lifetime=1.5
    Position Mode=::NewEnumerator0
    Sprite Size=&Multiply_Vector2D_ByFloat(
     Float=&FN_FrontEndFloatModifier(
      Gameplay=&RandomRangeFloat(Minimum=0.3, Maximum=0.45)
      Front End=&RandomRangeFloat(Minimum=0.3, Maximum=0.35)
     )
     Vector2D=<15 19.3113>
    )
    Sprite Size Mode=::NewEnumerator4
    Write Material Random=true
   )
   &AddVelocity(Module.Use Falloff Distance=false, Override Seed=false, Velocity Mode=::NewEnumerator1, Velocity Speed=100)
  ]
  particleUpdateScript [
   &ParticleState()
   &ScaleSpriteSize(
    Uniform Curve Scale=&Lerp_Float(
     A=1
     B=1.8
     Alpha=&FN_LOD_Distance_Remap(Start Distance=100, End Distance=1000, Minimum Output=0, Maximum Output=1, Clamp Output=true)
    )
    Uniform Curve Sprite Scale=DataInterface() {propertyValues {Curve {defaultValue=0, keys [{tangentMode=RCTM_User, time=0, value=0.4118}, <at 0.2094 1>, <at 0.323 1>, {tangentMode=RCTM_User, time=1, value=0}]}}}
   )
   &CurlNoiseForce(
    Bias Noise=false
    Noise Frequency=&MakeCustomFloatFromBool(Bool=true, True Float=34.1388, False Float=156.5192)
    Noise Strength=&FloatFromCurve(
     FloatCurve=DataInterface() {propertyValues {Curve {keys [<at 0 1>,<at 1 1>]}}}
     Scale Curve=&MakeCustomFloatFromBool(
      Bool=$User.bIsFrontEnd
      True Float=&RandomRangeFloat(Minimum=500, Maximum=1500)
      False Float=&RandomRangeFloat(Minimum=4000, Maximum=6000)
     )
    )
    Pan Noise=false
   )
   &ShapeLocation(
    MaskSpawnGroup=false
    Offset Coordinate Space=::World
    Offset Mode=::NewEnumerator0
    Rotation Angle:NiagaraFloat=&RandomRangeFloat(Minimum=0, Maximum=1)
    Rotation Axis=<1 1 1>
    Rotation Coordinate Space=::World
    Rotation Mode=::NewEnumerator1
    Sphere Distribution=::NewEnumerator1
    Sphere Radius=$User.Radius
    Sphere Radius Position:NiagaraFloat=&RandomRangeFloat(Minimum=0.63, Maximum=1)
    U Position=&MakeCustomFloatFromBool(
     Bool=&RandomBool()
     True Float=&FloatFromCurve(
      FloatCurve=DataInterface() {propertyValues {Curve {
         defaultValue=1
         keys [
          {arriveTangent=4.5812, interpMode=RCIM_Cubic, leaveTangent=4.5812, tangentMode=RCTM_User, time=0, value=0}
          {arriveTangent=0.9143, interpMode=RCIM_Cubic, leaveTangent=0.9143, tangentMode=RCTM_User, time=0.2561, value=0.7055}
          {arriveTangent=0.0224, interpMode=RCIM_Cubic, leaveTangent=0.0224, tangentMode=RCTM_User, time=1.0791, value=1}
         ]
        }}}
      CurveIndex=&Multiply_Float(A=$Particles.NormalizedAge, B=&RandomRangeFloat(Minimum=0.1, Maximum=0.5))
      Scale Curve=&RandomRangeFloat(Minimum=1, Maximum=2)
     )
     False Float=&FloatFromCurve(
      FloatCurve=DataInterface() {propertyValues {Curve {keys [{interpMode=RCIM_Cubic, leaveTangent=-1.7692, tangentMode=RCTM_Break, time=0, value=1}, <at 1 0>]}}}
      CurveIndex=&Multiply_Float(A=$Particles.NormalizedAge, B=&RandomRangeFloat(Minimum=0.1, Maximum=0.5))
      Scale Curve=&RandomRangeFloat(Minimum=1, Maximum=2)
     )
    )
    UseOwnerScale=true
    V Position=&MakeCustomFloatFromBool(
     Bool=&RandomBool()
     True Float=&FloatFromCurve(
      FloatCurve=DataInterface() {propertyValues {Curve {
         defaultValue=1
         keys [
          {arriveTangent=4.5812, interpMode=RCIM_Cubic, leaveTangent=4.5812, tangentMode=RCTM_User, time=0, value=0}
          {arriveTangent=1.6567, interpMode=RCIM_Cubic, leaveTangent=1.6567, tangentMode=RCTM_User, time=0.1811, value=0.6464}
          {arriveTangent=0.0224, interpMode=RCIM_Cubic, leaveTangent=0.0224, tangentMode=RCTM_User, time=1, value=1}
         ]
        }}}
      CurveIndex=&Multiply_Float(A=$Particles.NormalizedAge, B=&RandomRangeFloat(Minimum=0.1, Maximum=0.5))
     )
     False Float=&FloatFromCurve(
      FloatCurve=DataInterface() {propertyValues {Curve {keys [
          {interpMode=RCIM_Cubic, leaveTangent=-3.3122, tangentMode=RCTM_Break, time=0, value=1}
          {arriveTangent=-0.4912, tangentMode=RCTM_User, time=1, value=0}
         ]}}}
      CurveIndex=&Multiply_Float(A=$Particles.NormalizedAge, B=&RandomRangeFloat(Minimum=0.1, Maximum=0.5))
     )
    )
   )
   &ScaleVelocity(Velocity Scale=&VectorFromCurve(
     VectorCurve=DataInterface() {propertyValues {
       XCurve {keys [<at 0 0.2>, {interpMode=RCIM_Cubic, tangentMode=RCTM_Break, time=0.5, value=1}, <at 1 0.2>]}
       YCurve {keys [<at 0 0.2>, {interpMode=RCIM_Cubic, tangentMode=RCTM_Break, time=0.5, value=1}, <at 1 0.2>]}
       ZCurve {keys [<at 0 0.2>, {arriveTangent=0.4726, interpMode=RCIM_Cubic, tangentMode=RCTM_Break, time=0.5, value=1}, <at 1 0.2>]}
      }}
     Scale Curve=&RandomRangeVector(Minimum=<2 2 2>, Maximum=<3 3 3>)
    ))
   &Drag(Drag=&RandomRangeFloat(Minimum=1, Maximum=2), Use Linear Drag=true, Use Rotational Drag=false)
   &SolveForcesAndVelocity(Clamp Velocity=false, Limit Acceleration=false)
   &InterpolateOverTime(Rate Of Change=3, Target Float Value:NiagaraFloat=$NPC.FortniteNPC.IsPlayerIndoors)
   &ScaleColor(
    Scale RGB=&MergeHelper_VectorSwitch(
     0=&MakeVectorFromLinearColorRGB(Color=&LerpLinearColors(
       StartColor=<rgba 1 1 1 1>
       EndColor=<rgba 0.2031 0.2031 0.2031 1>
       LerpFactor=&FN_LOD_Distance_Remap(Start Distance=0, End Distance=1000, Minimum Output=0, Maximum Output=1, Clamp Output=true)
      ))
     1=&MakeVectorFromLinearColorRGB(Color=&LerpLinearColors(
       StartColor=<rgba 1 1 1 1>
       EndColor=<rgba 0.2031 0.2031 0.2031 1>
       LerpFactor=&FN_LOD_Distance_Remap(Start Distance=0, End Distance=1000, Minimum Output=0, Maximum Output=1, Clamp Output=true)
      ))
     2=&Lerp_Vector(
      A=&MakeVectorFromLinearColorRGB(Color=&LerpLinearColors(
        StartColor=&FNTimeOfDayColor(
         Night Color=<rgba 0.0769 0.0888 0.1354 0.9391>
         Morning Color=<rgba 0.7031 0.7031 0.7031 0.9391>
         Day Color=<rgba 0.7031 0.7031 0.7031 0.9391>
         Evening Color=<rgba 0.7031 0.7031 0.7031 0.9391>
        )
        EndColor=&FNTimeOfDayColor(
         Night Color=<rgba 0.0345 0.0304 0.099 0.9391>
         Morning Color=<rgba 0.0469 0.0469 0.0469 0.9391>
         Day Color=<rgba 0.0469 0.0469 0.0469 0.9391>
         Evening Color=<rgba 0.0469 0.0469 0.0469 0.9391>
        )
        LerpFactor=&FN_LOD_Distance_Remap(Start Distance=0, End Distance=1000, Minimum Output=0, Maximum Output=1, Clamp Output=true)
       ))
      B=&MakeVectorFromLinearColorRGB(Color=&LerpLinearColors(
        StartColor=&FNTimeOfDayColor(
         Night Color=<rgba 0.0385 0.0444 0.0677 0.9391>
         Morning Color=<rgba 0.0938 0.0938 0.0938 0.9391>
         Day Color=<rgba 0.0938 0.0938 0.0938 0.9391>
         Evening Color=<rgba 0.0938 0.0938 0.0938 0.9391>
        )
        EndColor=&FNTimeOfDayColor(
         Night Color=<rgba 0.0198 0.0184 0.0417 0.9391>
         Morning Color=<rgba 0.026 0.026 0.026 0.9391>
         Day Color=<rgba 0.026 0.026 0.026 0.9391>
         Evening Color=<rgba 0.026 0.026 0.026 0.9391>
        )
        LerpFactor=&FN_LOD_Distance_Remap(Start Distance=0, End Distance=1000, Minimum Output=0, Maximum Output=1, Clamp Output=true)
       ))
      Alpha=$Particles.InterpolateOverTime.InterpolateOverTime_Float.Moving Average
     )
     3=&Lerp_Vector(
      A=&MakeVectorFromLinearColorRGB(Color=&LerpLinearColors(
        StartColor=&FNTimeOfDayColor(
         Night Color=<rgba 0.0769 0.0888 0.1354 0.9391>
         Morning Color=<rgba 0.7031 0.7031 0.7031 0.9391>
         Day Color=<rgba 0.7031 0.7031 0.7031 0.9391>
         Evening Color=<rgba 0.7031 0.7031 0.7031 0.9391>
        )
        EndColor=&FNTimeOfDayColor(
         Night Color=<rgba 0.0345 0.0304 0.099 0.9391>
         Morning Color=<rgba 0.0469 0.0469 0.0469 0.9391>
         Day Color=<rgba 0.0469 0.0469 0.0469 0.9391>
         Evening Color=<rgba 0.0469 0.0469 0.0469 0.9391>
        )
        LerpFactor=&FN_LOD_Distance_Remap(Start Distance=0, End Distance=1000, Minimum Output=0, Maximum Output=1, Clamp Output=true)
       ))
      B=&MakeVectorFromLinearColorRGB(Color=&LerpLinearColors(
        StartColor=&FNTimeOfDayColor(
         Night Color=<rgba 0.0385 0.0444 0.0677 0.9391>
         Morning Color=<rgba 0.0938 0.0938 0.0938 0.9391>
         Day Color=<rgba 0.0938 0.0938 0.0938 0.9391>
         Evening Color=<rgba 0.0938 0.0938 0.0938 0.9391>
        )
        EndColor=&FNTimeOfDayColor(
         Night Color=<rgba 0.0198 0.0184 0.0417 0.9391>
         Morning Color=<rgba 0.026 0.026 0.026 0.9391>
         Day Color=<rgba 0.026 0.026 0.026 0.9391>
         Evening Color=<rgba 0.026 0.026 0.026 0.9391>
        )
        LerpFactor=&FN_LOD_Distance_Remap(Start Distance=0, End Distance=1000, Minimum Output=0, Maximum Output=1, Clamp Output=true)
       ))
      Alpha=$Particles.InterpolateOverTime.InterpolateOverTime_Float.Moving Average
     )
     4=&Lerp_Vector(
      A=&MakeVectorFromLinearColorRGB(Color=&LerpLinearColors(
        StartColor=&FNTimeOfDayColor(
         Night Color=<rgba 0.0769 0.0888 0.1354 0.9391>
         Morning Color=<rgba 0.7031 0.7031 0.7031 0.9391>
         Day Color=<rgba 0.7031 0.7031 0.7031 0.9391>
         Evening Color=<rgba 0.7031 0.7031 0.7031 0.9391>
        )
        EndColor=&FNTimeOfDayColor(
         Night Color=<rgba 0.0345 0.0304 0.099 0.9391>
         Morning Color=<rgba 0.0469 0.0469 0.0469 0.9391>
         Day Color=<rgba 0.0469 0.0469 0.0469 0.9391>
         Evening Color=<rgba 0.0469 0.0469 0.0469 0.9391>
        )
        LerpFactor=&FN_LOD_Distance_Remap(Start Distance=0, End Distance=1000, Minimum Output=0, Maximum Output=1, Clamp Output=true)
       ))
      B=&MakeVectorFromLinearColorRGB(Color=&LerpLinearColors(
        StartColor=&FNTimeOfDayColor(
         Night Color=<rgba 0.0385 0.0444 0.0677 0.9391>
         Morning Color=<rgba 0.0938 0.0938 0.0938 0.9391>
         Day Color=<rgba 0.0938 0.0938 0.0938 0.9391>
         Evening Color=<rgba 0.0938 0.0938 0.0938 0.9391>
        )
        EndColor=&FNTimeOfDayColor(
         Night Color=<rgba 0.0198 0.0184 0.0417 0.9391>
         Morning Color=<rgba 0.026 0.026 0.026 0.9391>
         Day Color=<rgba 0.026 0.026 0.026 0.9391>
         Evening Color=<rgba 0.026 0.026 0.026 0.9391>
        )
        LerpFactor=&FN_LOD_Distance_Remap(Start Distance=0, End Distance=1000, Minimum Output=0, Maximum Output=1, Clamp Output=true)
       ))
      Alpha=$Particles.InterpolateOverTime.InterpolateOverTime_Float.Moving Average
     )
     Index=$Engine.QualityLevel
     Num=::NewEnumerator3
    )
    ScaleA=true
    ScaleRGB=true
   )
   /SetVariables(Particles.SpriteAlign:Vector3f=&Subtract_Position(A=$Particles.Position, B=$Particles.Previous.Position)) {isSetParameters=true}
   &SubUVAnimation(
    "Sprite Renderer "=DataInterface() {propertyValues {SpriteRenderer=&self:DirectionalBurst.NiagaraSpriteRendererProperties_0}}
    SubUV Lookup Index:NiagaraFloat=&Sine(Normalized Angle=$Particles.Age, Period=0.344, Scale=1, Bias=0)
    SubUV Lookup Index Scale=3.3916
    UseEndFrame=false
    UseStartFrame=false
   )
  ]
  rendererClasses [&NiagaraSpriteRendererProperties]
  renderers [{rendererClass=&NiagaraSpriteRendererProperties}]
 }
}
summary {
 systemName=NS_Emote_BasketSuburb_AmbientFlies
 emitters [
  {emitterName=DirectionalBurst, rendererClasses [&NiagaraSpriteRendererProperties]}
  {emitterName=BlowingParticles, rendererClasses [&NiagaraRibbonRendererProperties]}
 ]
 userVariables [
  User.bIsFrontEnd {type=NiagaraBool}
  User.bIsFrontEndPreview {type=NiagaraBool}
  User.Radius {type=NiagaraFloat, defaultValue=50, description=""}
 ]
}
