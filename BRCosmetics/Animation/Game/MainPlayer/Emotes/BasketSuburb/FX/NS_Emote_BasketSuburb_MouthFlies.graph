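# Niagara effect for the BasketSuburb emote, named for flies at the mouth, spraying short-lived animated sprites in a cone.
# DirectionalBurst jitters them with curl noise, point force and drag, inheriting owner velocity conditionally.
# Ribbons reads those particles to draw faint trails.
# Color follows time of day and darkens indoors at higher quality levels, sprites grow and fade with distance, and a debug emitter is disabled.
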NiagaraSystem /BRCosmetics/Animation/Game/MainPlayer/Emotes/BasketSuburb/FX/NS_Emote_BasketSuburb_MouthFlies
dependencies {
 usedDataInterfaces [
  &NiagaraDataInterfaceCurve
  &NiagaraDataInterfaceSpriteRendererInfo
  &NiagaraDataInterfaceParticleRead
  &NiagaraDataInterfaceColorCurve
  &NiagaraDataInterfaceSocketReader
 ]
 usedDynamicInputs [
  &MakeVector4_FromLinearColor
  &LerpLinearColors
  /Niagara/DynamicInputs/LinearColor/MakeLinearColorFromVector4
  &Vector4FromCurve
 ]
 usedModules [
  &SystemState
  &self:SystemSpawnScript.SystemScriptSource.SystemScriptGraph.NiagaraNodeAssignment_0.SetVariables_#1
  &self:SystemSpawnScript.SystemScriptSource.SystemScriptGraph.NiagaraNodeAssignment_1.SetVariables_#2
 ]
}
emitters {
 DirectionalBurst {
  emitterUpdateScript [
   &EmitterState(Life Cycle Mode=::NewEnumerator1, Loop Behavior=::NewEnumerator1, Loop Duration Mode=::NewEnumerator1, UseLoopDelay=false)
   &SpawnRate(SpawnRate=6, Use Spawn Probability=false)
  ]
  particleSpawnScript [
   &InitializeParticle(
    Lifetime=&FN_FrontEndFloatModifier(Gameplay=0.6704, Front End=0.4)
    Position Mode=::NewEnumerator0
    Sprite Size=&Multiply_Vector2D_ByFloat(
     Float=&FN_FrontEndFloatModifier(
      Gameplay=&RandomRangeFloat(Minimum=0.3, Maximum=0.45)
      Front End=&RandomRangeFloat(Minimum=0.3, Maximum=0.35)
     )
     Vector2D=<15 19.3113>
    )
    Sprite Size Mode=::NewEnumerator4
    Write Material Random=true
   )
   &AddVelocity(
    Cone Angle=70.4356
    Cone Axis=<0 1 0>
    Module.UseConeFalloff=false
    Module.UseVelDistribution=false
    Override Seed=false
    Velocity Mode=::NewEnumerator2
    Velocity Speed=&RandomRangeFloat(Minimum=270, Maximum=350)
   )
   &ShapeLocation(Box Size=<50 50 2>, MaskSpawnGroup=false, Override Seed=false, Shape Primitive=::NewEnumerator2, Surface Only=false, UseOwnerScale=true) {module=ShapeLocation001}
  ]
  particleUpdateScript [
   &ParticleState()
   &CurlNoiseForce(
    Bias Noise=false
    Noise Frequency=1036.1106
    Noise Strength=&FloatFromCurve(
     FloatCurve=DataInterface() {propertyValues {Curve {defaultValue=1, keys [{tangentMode=RCTM_User, time=0, value=0}, {tangentMode=RCTM_User, time=0.4123, value=1}]}}}
     Scale Curve=2045.6135
    )
    Pan Noise=false
   )
   &PointForce(
    Force Origin=&AddVectorToPosition(Position=/Niagara/DynamicInputs/Helpers/SimulationPosition(), Vector=<0 63.0715 0>)
    Force Strength=81.2961
    Module.Use Falloff=false
    Module.UseLinearForce=false
   )
   &Drag(
    Drag=&FloatFromCurve(
     FloatCurve=DataInterface() {propertyValues {Curve {
        defaultValue=1
        keys [
         {arriveTangent=0.0882, interpMode=RCIM_Cubic, leaveTangent=0.0882, tangentMode=RCTM_User, time=0, value=0}
         {arriveTangent=16.1111, interpMode=RCIM_Cubic, leaveTangent=16.1111, tangentMode=RCTM_User, time=0.1904, value=1}
        ]
       }}}
     Scale Curve=17.0863
    )
    Use Linear Drag=true
    Use Rotational Drag=true
   )
   &InheritSourceMovement(
    Applied Position Scale=&MakeCustomVectorFromBool(
     Bool=&SetBoolByFloatComparison(A=&VectorLength(Vector=$User.OwnerLocalVelocity), B=350)
     True Vector=<0.5 0.5 0.5>
     False Vector=<0 0 0>
    )
    ApplyForce=false
    ApplyPosition=true
    ApplyPositionFalloff=false
    Source Velocity:Vector3f=$User.OwnerLocalVelocity
   )
   &SolveForcesAndVelocity(Clamp Velocity=false, Limit Acceleration=false)
   &ScaleSpriteSize(
    Uniform Curve Scale=&Lerp_Float(
     A=1
     B=1.8
     Alpha=&FN_LOD_Distance_Remap(Start Distance=100, End Distance=1000, Minimum Output=0, Maximum Output=1, Clamp Output=true)
    )
    Uniform Curve Sprite Scale=DataInterface() {propertyValues {Curve {defaultValue=0, keys [{tangentMode=RCTM_User, time=0, value=0.4118}, <at 0.2094 1>, <at 0.323 1>, {tangentMode=RCTM_User, time=1, value=0}]}}}
   )
   &SubUVAnimation(
    "Sprite Renderer "=DataInterface() {propertyValues {SpriteRenderer=&self:DirectionalBurst.NiagaraSpriteRendererProperties_0}}
    SubUV Lookup Index:NiagaraFloat=&Sine(Normalized Angle=$Particles.Age, Period=0.344, Scale=1, Bias=0)
    UseEndFrame=false
    UseStartFrame=false
   )
   &InterpolateOverTime(Rate Of Change=3, Target Float Value:NiagaraFloat=$NPC.FortniteNPC.IsPlayerIndoors)
   &ScaleColor(
    Scale Alpha=&Lerp_Float(
     A=1
     B=0
     Alpha=&FN_LOD_Distance_Remap(Start Distance=0, End Distance=1000, Minimum Output=0, Maximum Output=1, Clamp Output=true)
    )
    Scale RGB=&MergeHelper_VectorSwitch(
     0=&MakeVectorFromLinearColorRGB(Color=&FNTimeOfDayColor(
       Night Color=<rgba 0.4024 0.4646 0.7083 0.9391>
       Morning Color=<rgba 0.7031 0.7031 0.7031 0.9391>
       Day Color=<rgba 0.7031 0.7031 0.7031 0.9391>
       Evening Color=<rgba 0.7031 0.7031 0.7031 0.9391>
      ))
     1=&MakeVectorFromLinearColorRGB(Color=&FNTimeOfDayColor(
       Night Color=<rgba 0.4024 0.4646 0.7083 0.9391>
       Morning Color=<rgba 0.7031 0.7031 0.7031 0.9391>
       Day Color=<rgba 0.7031 0.7031 0.7031 0.9391>
       Evening Color=<rgba 0.7031 0.7031 0.7031 0.9391>
      ))
     2=&Lerp_Vector(
      A=&MakeVectorFromLinearColorRGB(Color=&FNTimeOfDayColor(
        Night Color=<rgba 0.0769 0.0888 0.1354 0.9391>
        Morning Color=<rgba 0.7031 0.7031 0.7031 0.9391>
        Day Color=<rgba 0.7031 0.7031 0.7031 0.9391>
        Evening Color=<rgba 0.7031 0.7031 0.7031 0.9391>
       ))
      B=&MakeVectorFromLinearColorRGB(Color=&FNTimeOfDayColor(
        Night Color=<rgba 0.0385 0.0444 0.0677 0.9391>
        Morning Color=<rgba 0.0938 0.0938 0.0938 0.9391>
        Day Color=<rgba 0.0938 0.0938 0.0938 0.9391>
        Evening Color=<rgba 0.0938 0.0938 0.0938 0.9391>
       ))
      Alpha=$Particles.InterpolateOverTime.InterpolateOverTime_Float.Moving Average
     )
     3=&Lerp_Vector(
      A=&MakeVectorFromLinearColorRGB(Color=&FNTimeOfDayColor(
        Night Color=<rgba 0.0769 0.0888 0.1354 0.9391>
        Morning Color=<rgba 0.7031 0.7031 0.7031 0.9391>
        Day Color=<rgba 0.7031 0.7031 0.7031 0.9391>
        Evening Color=<rgba 0.7031 0.7031 0.7031 0.9391>
       ))
      B=&MakeVectorFromLinearColorRGB(Color=&FNTimeOfDayColor(
        Night Color=<rgba 0.0385 0.0444 0.0677 0.9391>
        Morning Color=<rgba 0.0938 0.0938 0.0938 0.9391>
        Day Color=<rgba 0.0938 0.0938 0.0938 0.9391>
        Evening Color=<rgba 0.0938 0.0938 0.0938 0.9391>
       ))
      Alpha=$Particles.InterpolateOverTime.InterpolateOverTime_Float.Moving Average
     )
     4=&Lerp_Vector(
      A=&MakeVectorFromLinearColorRGB(Color=&FNTimeOfDayColor(
        Night Color=<rgba 0.0769 0.0888 0.1354 0.9391>
        Morning Color=<rgba 0.7031 0.7031 0.7031 0.9391>
        Day Color=<rgba 0.7031 0.7031 0.7031 0.9391>
        Evening Color=<rgba 0.7031 0.7031 0.7031 0.9391>
       ))
      B=&MakeVectorFromLinearColorRGB(Color=&FNTimeOfDayColor(
        Night Color=<rgba 0.0385 0.0444 0.0677 0.9391>
        Morning Color=<rgba 0.0938 0.0938 0.0938 0.9391>
        Day Color=<rgba 0.0938 0.0938 0.0938 0.9391>
        Evening Color=<rgba 0.0938 0.0938 0.0938 0.9391>
       ))
      Alpha=$Particles.InterpolateOverTime.InterpolateOverTime_Float.Moving Average
     )
     Index=$Engine.QualityLevel
     Num=::NewEnumerator3
    )
    ScaleA=true
    ScaleRGB=true
   )
   /SetVariables(Particles.Vector:Vector3f=&Lerp_Vector(
     A=&MakeVectorFromLinearColorRGB(Color=&FNTimeOfDayColor(
       Night Color=<rgba 0.0769 0.0888 0.1354 0.9391>
       Morning Color=<rgba 0.7031 0.7031 0.7031 0.9391>
       Day Color=<rgba 0.7031 0.7031 0.7031 0.9391>
       Evening Color=<rgba 0.7031 0.7031 0.7031 0.9391>
      ))
     B=&MakeVectorFromLinearColorRGB(Color=&FNTimeOfDayColor(
       Night Color=<rgba 0.0385 0.0444 0.0677 0.9391>
       Morning Color=<rgba 0.0938 0.0938 0.0938 0.9391>
       Day Color=<rgba 0.0938 0.0938 0.0938 0.9391>
       Evening Color=<rgba 0.0938 0.0938 0.0938 0.9391>
      ))
     Alpha=$Particles.InterpolateOverTime.InterpolateOverTime_Float.Moving Average
    )) {isSetParameters=true}
  ]
  rendererClasses [&NiagaraSpriteRendererProperties]
  renderers [{rendererClass=&NiagaraSpriteRendererProperties}]
 }
 NE_Debug_Float {
  bEnabled=false
  emitterSpawnScript [/SetVariables(Emitter.Digits=2, Emitter.DisplayScalar=true) {isSetParameters=true}]
  emitterUpdateScript [
   &EmitterState()
   &SpawnBurst_Instantaneous(Use Spawn Probability=false, UseLoopCountLimit=false)
  ]
  particleSpawnScript [&InitializeParticle(
    Color Mode=::NewEnumerator0
    Lifetime=$Emitter.CurrentLoopDuration
    Position Offset=<0 0 70>
    Sprite Size Mode=::NewEnumerator3
    Uniform Sprite Size=125
    UsePositionOffset=true
    Write Material Random=true
   )]
  particleUpdateScript [
   &ParticleState(Kill Particles When Lifetime Has Elapsed=false)
   &InterpolateOverTime(Rate Of Change=3, Target Float Value:NiagaraFloat=$System.float)
   &Color(Color=/Niagara/DynamicInputs/TypeConversions/MakeLinearColorFromFloat(Value:NiagaraFloat=$Particles.InterpolateOverTime.InterpolateOverTime_Float.Moving Average))
  ]
  rendererClasses [&NiagaraSpriteRendererProperties]
  renderers [{rendererClass=&NiagaraSpriteRendererProperties}]
 }
 Ribbons {
  emitterUpdateScript [
   /SetVariables(Emitter.NewSpawnRate=16) {isSetParameters=true}
   &EmitterState(Life Cycle Mode=::NewEnumerator1, Loop Duration=0.0056, UseLoopDelay=false)
   &SpawnParticlesFromOtherEmitter(
    Attribute Reader=DataInterface() {propertyValues {EmitterBinding {bindingMode=Other, emitterName=DirectionalBurst}}}
    Maximum Age Spawn Threshold=1.4
    Spawn Rate=$Emitter.NewSpawnRate
    Use Maximum Age=false
    Use Minimum Age=false
   )
  ]
  particleSpawnScript [
   &InitializeParticle(
    Color=&FNTimeOfDayColor(Night Color=<rgba 0.4669 0.5366 1 0.113>, Morning Color=<rgba 1 1 1 0.113>, Day Color=<rgba 1 1 1 0.113>, Evening Color=<rgba 1 1 1 0.113>)
    Lifetime=0.2
    Ribbon Width=1.0428
    Ribbon Width Mode=::NewEnumerator1
    UsePositionOffset=false
    Write Material Random=true
   )
   &SampleParticlesFromOtherEmitter(
    Particle ID Sampling=::NewEnumerator1
    Reject Particle If Sampled Index Is Invalid=false
    Sprite Size Sampling:ENiagara_DirectReadSamplingMode=::NewEnumerator1
   )
  ]
  particleUpdateScript [
   &ParticleState()
   /Niagara/Modules/AttributeReader/UpdateParticlesFromOtherEmitter(
    Attribute Reader=DataInterface() {propertyValues {EmitterBinding {bindingMode=Other, emitterName=DirectionalBurst}}}
    Position Sampling:ENiagara_DirectReadSamplingMode=::NewEnumerator1
    Sprite Size Sampling:ENiagara_DirectReadSamplingMode=::NewEnumerator1
   )
   &self:ScratchModule(Position:NiagaraPosition=$Output.UpdateParticlesFromOtherEmitter.SampledPosition) {module=ScratchModule}
   &ScaleRibbonWidth(Ribbon Width Scale=&Multiply_Float(A=0.4, B=&MakeFloatFromVector2d(Vector 2D=$Output.UpdateParticlesFromOtherEmitter.SampledSpriteSize)))
   &InheritSourceMovement(Applied Position Scale=<0.3 0.3 0.3>, ApplyForce=false, ApplyPosition=true, ApplyPositionFalloff=false)
   &SolveForcesAndVelocity(Clamp Velocity=false, Limit Acceleration=false)
   &InterpolateOverTime(
    Rate Of Change=5
    Target Float Value:NiagaraFloat=&RemapRange(Input Value To Remap=$NPC.FortniteNPC.IsPlayerIndoors, Input Min=0, Input Max=0.2, Output Min=0, Output Max=1)
   )
   &ScaleColor(
    Scale Alpha=&FloatFromCurve(FloatCurve=DataInterface() {propertyValues {Curve {keys [
         {arriveTangent=-0.8321, interpMode=RCIM_Cubic, leaveTangent=35.6275, tangentMode=RCTM_Break, time=0, value=0}
         {arriveTangent=0.0819, interpMode=RCIM_Cubic, leaveTangent=0.6334, tangentMode=RCTM_Break, time=0.0597, value=1}
         {arriveTangent=-1.7699, tangentMode=RCTM_User, time=1, value=0}
        ]}}})
    Scale RGB=&MergeHelper_VectorSwitch(
     0=&MakeVectorFromLinearColorRGB(Color=<rgba 1 1 1 1>)
     1=&MakeVectorFromLinearColorRGB(Color=<rgba 1 1 1 1>)
     2=&Lerp_Vector(
      A=&MakeVectorFromLinearColorRGB(Color=<rgba 1 1 1 1>)
      B=&MakeVectorFromLinearColorRGB(Color=<rgba 0.1719 0.1719 0.1719 1>)
      Alpha=$Particles.InterpolateOverTime.InterpolateOverTime_Float.Moving Average
     )
     3=&Lerp_Vector(
      A=&MakeVectorFromLinearColorRGB(Color=<rgba 1 1 1 1>)
      B=&MakeVectorFromLinearColorRGB(Color=<rgba 0.1719 0.1719 0.1719 1>)
      Alpha=$Particles.InterpolateOverTime.InterpolateOverTime_Float.Moving Average
     )
     4=&Lerp_Vector(
      A=&MakeVectorFromLinearColorRGB(Color=<rgba 1 1 1 1>)
      B=&MakeVectorFromLinearColorRGB(Color=<rgba 0.1719 0.1719 0.1719 1>)
      Alpha=$Particles.InterpolateOverTime.InterpolateOverTime_Float.Moving Average
     )
     Index=$Engine.QualityLevel
     Num=::NewEnumerator3
    )
    ScaleA=true
    ScaleRGB=true
   )
  ]
  rendererClasses [&NiagaraRibbonRendererProperties]
  renderers [{rendererClass=&NiagaraRibbonRendererProperties}]
 }
}
summary {
 systemName=NS_Emote_BasketSuburb_MouthFlies
 emitters [
  {emitterName=DirectionalBurst, rendererClasses [&NiagaraSpriteRendererProperties]}
  {emitterName=Ribbons, rendererClasses [&NiagaraRibbonRendererProperties]}
  {bEnabled=false, emitterName=NE_Debug_Float, rendererClasses [&NiagaraSpriteRendererProperties]}
 ]
 userVariables [
  User.bIsFrontEnd {type=NiagaraBool}
  User.bIsFrontEndPreview {type=NiagaraBool}
  User.OwnerLocalVelocity {type=Vector3f, description="", defaultValue {struct=Vector3f, value=<0 0 0>}}
 ]
}
